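# Sound cue for the energy barrel explosion, layering keg impact, watery explode and slightly delayed splash sounds.
# Each layer picks a random wave with volume or pitch variation, and a mixer blends the three.
# A quality-level switch feeds the top two levels from the fuller mixer and the lower two from a copy with fewer variations.
# It plays on the impacts sound class with overridden attenuation and a prevent-new concurrency rule.
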SoundCue /EnergyProps/BGAs/EnergyBarrel/Sound/Energy_Barrel_Explode_Cue
root=QualityLevel_0
nodes [
 {id=Delay_0, children [Modulator_2], properties {delayMax=0.1, delayMin=0.1}}
 {id=Delay_2, children [Modulator_6], properties {delayMax=0.1, delayMin=0.1}}
 {id=Mixer_0, children [Modulator_4,Modulator_7,Delay_2], properties {inputVolume [0.6,1,0.8]}}
 {id=Mixer_1, children [Modulator_0,Modulator_3,Delay_0], properties {inputVolume [0.6,1,0.8]}}
 {id=Modulator_0, children [Random_2], properties {pitchMax=1.2, pitchMin=1.1, volumeMax=1.1, volumeMin=0.9}}
 {id=Modulator_1, children [Random_1], properties {pitchMax=1.1, pitchMin=0.9, volumeMax=1.1, volumeMin=0.9}}
 {id=Modulator_2, children [Random_0], properties {pitchMax=1.2, pitchMin=1.1}}
 {id=Modulator_3, children [Modulator_1], properties {pitchMax=1.2, pitchMin=1.1, volumeMax=1.1, volumeMin=0.9}}
 {id=Modulator_4, children [Random_4], properties {pitchMax=1.2, pitchMin=1.1, volumeMax=1.1, volumeMin=0.9}}
 {id=Modulator_5, children [Random_3], properties {pitchMax=1.1, pitchMin=0.9, volumeMax=1.1, volumeMin=0.9}}
 {id=Modulator_6, children [Random_5], properties {pitchMax=1.2, pitchMin=1.1}}
 {id=Modulator_7, children [Modulator_5], properties {pitchMax=1.2, pitchMin=1.1, volumeMax=1.1, volumeMin=0.9}}
 {id=QualityLevel_0, children [Mixer_1,Mixer_1,Mixer_0,Mixer_0]}
 {id=Random_0, children [WavePlayer_1,WavePlayer_2,WavePlayer_3,WavePlayer_4], properties {weights [1,1,1,1]}}
 {id=Random_1, children [WavePlayer_0,WavePlayer_5,WavePlayer_6], properties {weights [1,1,1]}}
 {
  id=Random_2
  children [WavePlayer_7, WavePlayer_8, WavePlayer_9, WavePlayer_10, WavePlayer_11, WavePlayer_12]
  properties {weights [1,1,1,1,1,1]}
 }
 {id=Random_3, children [WavePlayer_13, WavePlayer_14], properties {weights [1,1]}}
 {id=Random_4, children [WavePlayer_16, WavePlayer_17, WavePlayer_18], properties {weights [1,1,1]}}
 {id=Random_5, children [WavePlayer_22, WavePlayer_23], properties {weights [1,1]}}
 {id=WavePlayer_0, properties {soundWaveAssetPtr=&/Energy_Keg_Explode_Water_01}}
 {id=WavePlayer_1, properties {soundWaveAssetPtr=&/Energy_Keg_Splash_01}}
 {id=WavePlayer_10, properties {soundWaveAssetPtr=&/Energy_Keg_Impact_04}}
 {id=WavePlayer_11, properties {soundWaveAssetPtr=&/Energy_Keg_Impact_05}}
 {id=WavePlayer_12, properties {soundWaveAssetPtr=&/Energy_Keg_Impact_06}}
 {id=WavePlayer_13, properties {soundWaveAssetPtr=&Energy_Keg_Explode_Water_01}}
 {id=WavePlayer_14, properties {soundWaveAssetPtr=&/Energy_Keg_Explode_Water_02}}
 {id=WavePlayer_16, properties {soundWaveAssetPtr=&/Energy_Keg_Impact_01}}
 {id=WavePlayer_17, properties {soundWaveAssetPtr=&/Energy_Keg_Impact_02}}
 {id=WavePlayer_18, properties {soundWaveAssetPtr=&/Energy_Keg_Impact_03}}
 {id=WavePlayer_2, properties {soundWaveAssetPtr=&/Energy_Keg_Splash_02}}
 {id=WavePlayer_22, properties {soundWaveAssetPtr=&Energy_Keg_Splash_01}}
 {id=WavePlayer_23, properties {soundWaveAssetPtr=&Energy_Keg_Splash_02}}
 {id=WavePlayer_3, properties {soundWaveAssetPtr=&/Energy_Keg_Splash_03}}
 {id=WavePlayer_4, properties {soundWaveAssetPtr=&/Energy_Keg_Splash_04}}
 {id=WavePlayer_5, properties {soundWaveAssetPtr=&Energy_Keg_Explode_Water_02}}
 {id=WavePlayer_6, properties {soundWaveAssetPtr=&/Energy_Keg_Explode_Water_03}}
 {id=WavePlayer_7, properties {soundWaveAssetPtr=&Energy_Keg_Impact_01}}
 {id=WavePlayer_8, properties {soundWaveAssetPtr=&Energy_Keg_Impact_02}}
 {id=WavePlayer_9, properties {soundWaveAssetPtr=&Energy_Keg_Impact_03}}
]
properties {
 attenuationSettings=/Game/Sounds/Attenuation/Fort_Grenade_Smoke_Att
 bExcludeFromRandomNodeBranchCulling=true
 bOverrideAttenuation=true
 bOverrideConcurrency=true
 maxDistance=3000
 soundClassObject=&Fort_Impacts
 volumeMultiplier=1.32
 attenuationOverrides {falloffDistance=2400, attenuationShapeExtents {x=600}}
 concurrencyOverrides {resolutionRule=PreventNew}
}
