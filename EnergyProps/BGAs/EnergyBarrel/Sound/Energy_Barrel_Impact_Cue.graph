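# Sound cue that plays a randomized impact sound for the energy barrel, in the impacts sound class.
# A quality-level node routes its first two slots to a six-variation random pool and its last two to a three-variation pool.
# Each pool passes through a modulator that slightly varies pitch and volume.
# The cue uses its own attenuation and concurrency overrides, refusing new instances at the limit, and is excluded from random-branch culling.
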SoundCue /EnergyProps/BGAs/EnergyBarrel/Sound/Energy_Barrel_Impact_Cue
root=QualityLevel_0
nodes [
 {id=Modulator_0, children [Random_0], properties {pitchMax=1.1, pitchMin=1, volumeMax=1.3, volumeMin=1.2}}
 {id=Modulator_1, children [Random_1], properties {pitchMax=1.1, pitchMin=1, volumeMax=1.3, volumeMin=1.2}}
 {id=QualityLevel_0, children [Modulator_0,Modulator_0,Modulator_1,Modulator_1]}
 {
  id=Random_0
  children [WavePlayer_1, WavePlayer_2, WavePlayer_3, WavePlayer_4, WavePlayer_10, WavePlayer_11]
  properties {weights [1,1,1,1,1,1]}
 }
 {id=Random_1, children [WavePlayer_0,WavePlayer_5,WavePlayer_6], properties {weights [1,1,1]}}
 {id=WavePlayer_0, properties {soundWaveAssetPtr=&/Energy_Keg_Impact_01}}
 {id=WavePlayer_1, properties {soundWaveAssetPtr=&Energy_Keg_Impact_01}}
 {id=WavePlayer_10, properties {soundWaveAssetPtr=&/Energy_Keg_Impact_05}}
 {id=WavePlayer_11, properties {soundWaveAssetPtr=&/Energy_Keg_Impact_06}}
 {id=WavePlayer_2, properties {soundWaveAssetPtr=&/Energy_Keg_Impact_02}}
 {id=WavePlayer_3, properties {soundWaveAssetPtr=&/Energy_Keg_Impact_03}}
 {id=WavePlayer_4, properties {soundWaveAssetPtr=&/Energy_Keg_Impact_04}}
 {id=WavePlayer_5, properties {soundWaveAssetPtr=&Energy_Keg_Impact_02}}
 {id=WavePlayer_6, properties {soundWaveAssetPtr=&Energy_Keg_Impact_03}}
]
properties {
 attenuationSettings=/Game/Sounds/Attenuation/Fort_Demolition_Hit_Small_Sized_Object_Att
 bExcludeFromRandomNodeBranchCulling=true
 bOverrideAttenuation=true
 bOverrideConcurrency=true
 maxDistance=3000
 soundClassObject=&Fort_Impacts
 volumeMultiplier=0.84
 attenuationOverrides {falloffDistance=2400, attenuationShapeExtents {x=600}}
 concurrencyOverrides {resolutionRule=PreventNew}
}
